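# Editor material for preview meshes in the brick library thumbnail generator, Default Lit and enabled for Nanite.
# An eyeballed constant grey is split into channels, each converted from sRGB to linear, then recombined into a Float3.
# The graph shows no link from that result to a material output.
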Material /AtomFramework/Editor/Blueprints/BrickLibraryThumbnailGenerator/M_ThumbnailPreviewMeshMaterial
graph {
 properties {bUsedWithNanite=true, usedShadingModels=Default Lit}
 nodes [
  n1 = MaterialFunctionCall(Float3=Constant() {desc=Eyeballed replacement, r=0.9}) {materialFunction=&BreakOutFloat3Components}
  MaterialFunctionCall(
   X=MaterialFunctionCall!(sRGB=@n1.R) {materialFunction=/Engine/Functions/Engine_MaterialFunctions01/ImageAdjustment/sRGBToLinear}.Linear
   Y=MaterialFunctionCall(sRGB=@n1.G) {materialFunction=&sRGBToLinear}.Linear
   Z=MaterialFunctionCall(sRGB=@n1.B) {materialFunction=&sRGBToLinear}.Linear
  ) {materialFunction=&MakeFloat3}
 ]
}
