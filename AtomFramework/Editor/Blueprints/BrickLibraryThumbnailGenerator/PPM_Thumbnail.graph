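# Post-process material that gives brick library thumbnails a cel-shaded, black-outlined look.
# Stored base color is multiplied by stepped lighting from a fixed direction against world normals; pixels beyond a depth cutoff get a near-black vertical gradient.
# Outlines combine a four-neighbor depth difference with an eight-neighbor normal comparison, then blend the line color over the result.
# Several branches, including custom stencil and line-width scaling, feed nothing downstream.
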Material /AtomFramework/Editor/Blueprints/BrickLibraryThumbnailGenerator/PPM_Thumbnail
graph {
 properties {
  bIsBlendable=false
  blendableOutputAlpha=true
  materialDomain=MD_PostProcess
  pixelDepthOffsetMode=PDOM_Legacy
  usedShadingModels=Default Lit
 }
 nodes [
  n1 = ComponentMask(None=SceneTexture() {sceneTextureId=PPI_CustomStencil}.Color) {r=true}
  Clamp(None=MaterialFunctionCall!(Bit=Constant() {r=1}, BitMask=Divide(A=@n1) {constB=255}) {materialFunction=/Engine/Functions/Engine_MaterialFunctions01/Texturing/BitMask}.Result)
  n2 = SceneTexture() {sceneTextureId=PPI_SceneDepth}
  n3 = ComponentMask(None=ScreenPosition().ViewportUV) {g=true, r=true}
  n4 = SceneTexelSize()
  n5 = Multiply(A=@n4, B=Constant() {r=1})
  n6 = Constant()
  n7 = AppendVector(A=ComponentMask(None=@n5) {desc=X, r=true}, B=@n6)
  n8 = Add(A=@n3, B=@n7) {desc=X+}
  n9 = AppendVector(A=@n6, B=ComponentMask(None=@n5) {desc=Y, g=true})
  n10 = Add(A=@n3, B=@n9) {desc=Y+}
  n11 = Add(
   A=Subtract(A=@n2.Color, B=SceneTexture(UVs=@n8) {sceneTextureId=PPI_SceneDepth}.Color)
   B=Subtract(A=@n2.Color, B=SceneTexture(UVs=@n10) {sceneTextureId=PPI_SceneDepth}.Color)
  )
  n12 = Constant() {r=-1}
  n13 = Add(A=@n3, B=Multiply(A=@n7, B=@n12)) {desc=X-}
  n14 = Add(A=@n3, B=Multiply(A=@n12, B=@n9)) {desc=Y-}
  n15 = Add(
   A=Subtract(A=@n2.Color, B=SceneTexture(UVs=@n13) {sceneTextureId=PPI_SceneDepth}.Color)
   B=Subtract(A=@n2.Color, B=SceneTexture(UVs=@n14) {sceneTextureId=PPI_SceneDepth}.Color)
  )
  n16 = Divide(A=Abs(None=Add(A=@n11, B=@n15)), B=Constant() {desc=Edge Angle Falloff, r=1})
  n17 = Clamp(None=@n16)
  n18 = If(A=ComponentMask(None=SceneTexture() {sceneTextureId=PPI_SceneDepth}.Color) {r=true}, "A > B"=Constant() {r=1}, "A < B"=Constant()) {constB=5000}
  n19 = OneMinus(None=@n18)
  Clamp(None=Add(A=@n17, B=@n19) {constB=0})
  n20 = Constant() {r=1}
  If(A=ComponentMask(None=SceneTexture() {sceneTextureId=PPI_SceneDepth}.Color) {r=true}, "A > B"=Constant(), "A == B"=@n20, "A < B"=@n20) {constB=10000}
  MaterialFunctionCall!(Mask Offset Value, Number Before White Result, Gradient) {materialFunction=/Engine/Functions/Engine_MaterialFunctions02/ValueStep}
  n21 = Multiply(A=OneMinus(None=@n17), B=Constant() {r=2})
  Multiply(A=OneMinus(None=Clamp(None=@n21)))
  n22 = ViewSize()
  n23 = Constant() {r=320}
  n24 = Ceil(None=Subtract(A=Divide(A=@n22, B=@n23), B=Constant() {r=2}))
  Multiply(A=Multiply(A=Constant() {desc=Line Render Width, r=1}, B=Clamp(None=@n24) {maxDefault=2, minDefault=1}), B=@n4)
  n25 = SceneTexture() {sceneTextureId=PPI_SceneDepth}
  n26 = Divide(A=Add(A=ComponentMask(None=@n25.Color) {r=true}, B=Constant() {r=-300}), B=Constant() {r=800})
  Multiply(A=OneMinus(None=Clamp(None=@n26) {maxDefault=0.25, minDefault=0.05}), B=Constant() {r=0.95})
  n27 = ViewSize()
  n28 = Constant() {r=320}
  n29 = Ceil(None=Subtract(A=Divide(A=@n27, B=@n28), B=Constant() {r=2}))
  n30 = SceneTexelSize()
  Multiply(A=Multiply(A=Constant() {desc=Line Render Width, r=1}, B=Clamp(None=@n29) {maxDefault=2, minDefault=1}), B=@n30)
  n31 = Normalize(VectorInput=Constant3Vector() {constant=<rgba -1.25 2.5 3.6 0>})
  n32 = Normalize(VectorInput=SceneTexture() {sceneTextureId=PPI_WorldNormal}.Color)
  n33 = Constant() {r=0.1667}
  n34 = Divide(A=ConstantBiasScale(None=DotProduct(A=@n31, B=@n32)), B=@n33) {constB=0.125}
  n35 = Multiply(A=Multiply(A=Floor(None=@n34), B=@n33) {constB=0.125}, B=SceneTexture() {sceneTextureId=PPI_StoredBaseColor}.Color)
  n36 = Multiply(A=Clamp(None=ComponentMask(None=@n35) {b=true, g=true, r=true}), B=@n19)
  n37 = OneMinus(None=ComponentMask(None=ScreenPosition().ViewportUV) {g=true})
  n38 = Constant() {r=0.02}
  n39 = Multiply(A=@n18, B=Add(A=Multiply(A=@n37, B=@n38)) {constB=0.02})
  n40 = Constant() {r=0.98}
  n41 = Constant()
  n42 = Constant() {r=1}
  n43 = Normalize(VectorInput=SceneTexture() {sceneTextureId=PPI_WorldNormal}.Color)
  n44 = ComponentMask(None=ScreenPosition().ViewportUV) {g=true, r=true}
  n45 = Multiply(A=@n30, B=Constant() {r=1})
  n46 = Constant()
  n47 = AppendVector(A=ComponentMask(None=@n45) {desc=X, r=true}, B=@n46)
  n48 = Add(A=@n44, B=@n47) {desc=X+}
  n49 = DotProduct(A=@n43, B=Normalize(VectorInput=SceneTexture(UVs=@n48) {sceneTextureId=PPI_WorldNormal}.Color))
  n50 = AppendVector(A=@n46, B=ComponentMask(None=@n45) {desc=Y, g=true})
  n51 = Add(A=@n44, B=@n50) {desc=Y+}
  n52 = DotProduct(A=@n43, B=Normalize(VectorInput=SceneTexture(UVs=@n51) {sceneTextureId=PPI_WorldNormal}.Color))
  n53 = Constant() {r=-1}
  n54 = Multiply(A=@n47, B=@n53)
  n55 = Add(A=@n44, B=@n54) {desc=X-}
  n56 = DotProduct(A=@n43, B=Normalize(VectorInput=SceneTexture(UVs=@n55) {sceneTextureId=PPI_WorldNormal}.Color))
  n57 = Multiply(A=@n53, B=@n50)
  n58 = Add(A=@n44, B=@n57) {desc=Y-}
  n59 = DotProduct(A=@n43, B=Normalize(VectorInput=SceneTexture(UVs=@n58) {sceneTextureId=PPI_WorldNormal}.Color))
  n60 = Add(A=@n44, B=Add(A=@n47, B=@n50))
  n61 = DotProduct(A=@n43, B=Normalize(VectorInput=SceneTexture(UVs=@n60) {sceneTextureId=PPI_WorldNormal}.Color))
  n62 = Add(A=@n44, B=Add(A=@n54, B=@n50))
  n63 = DotProduct(A=@n43, B=Normalize(VectorInput=SceneTexture(UVs=@n62) {sceneTextureId=PPI_WorldNormal}.Color))
  n64 = Add(A=@n44, B=Add(A=@n54, B=@n57))
  n65 = DotProduct(A=@n43, B=Normalize(VectorInput=SceneTexture(UVs=@n64) {sceneTextureId=PPI_WorldNormal}.Color))
  n66 = Add(A=@n44, B=Add(A=@n47, B=@n57))
  n67 = DotProduct(A=@n43, B=Normalize(VectorInput=SceneTexture(UVs=@n66) {sceneTextureId=PPI_WorldNormal}.Color))
  n68 = Add(A=Add(A=Add(A=@n49, B=@n52), B=Add(A=@n56, B=@n59)), B=Add(A=Add(A=@n61, B=@n63), B=Add(A=@n65, B=@n67)))
  n69 = Constant() {r=8}
  n70 = Add(
   A=OneMinus(None=If(A=@n16, B=@n40, "A > B"=@n41, "A < B"=@n42))
   B=If(
    A=Divide(A=@n68, B=@n69)
    B=Constant() {desc="Normal Change tolerance (.97 original)", r=0.988}
    "A > B"=Constant()
    "A < B"=Constant() {r=1}
   )
  )
  n71 = Add(A=Clamp(None=Clamp(None=@n70))) {constB=0}
  Reroute(None=LinearInterpolate(A=Add(A=@n36, B=@n39), B=Constant3Vector() {desc=Line color, constant=<rgba 0 0 0 1>}, Alpha=Clamp(None=@n71)))
  Transform(None=@n32) {transformSourceType=TRANSFORMSOURCE_World, transformType=TRANSFORM_Camera}
 ]
}
