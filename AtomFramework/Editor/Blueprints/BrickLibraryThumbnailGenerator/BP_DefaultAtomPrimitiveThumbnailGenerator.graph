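# Editor utility Blueprint that generates icon textures for Atom brick library primitives and writes them back into the data table.
# For each row it loads the mesh, applies a preview material and per-row offset, then frames it with an ortho scene capture sized from the mesh bounds.
# The render target becomes a static texture in an icon folder beside the table, named from the mesh name.
# A post-process material is loaded during setup.
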EditorUtilityBlueprint /AtomFramework/Editor/Blueprints/BrickLibraryThumbnailGenerator/BP_DefaultAtomPrimitiveThumbnailGenerator
graphs [
 CaptureThumbnail {nodes [
   CaptureThumbnail!() -> (then, StaticMeshAsset:Static Mesh Object Reference, AssetOffset:Transform, OutputDirectory:String) {is=title K2Node_FunctionEntry}
   > n1 = Sequence!(execute) -> (then_0, then_1, then_2:Exec, then_3:Exec, then_4:Exec, then_5:Exec)
   n2 = Get CaptureCamera() -> (CaptureCamera:Scene Capture Component 2D Object Reference)
   ShowOnlyComponent!(
    execute=@n1.then_2
    self:Scene Capture Component Object Reference=@n2
    InComponent:Primitive Component Object Reference=Get CaptureTarget() -> (CaptureTarget:Static Mesh Component Object Reference)
   ) -> (then)
   > CaptureScene(self:Scene Capture Component 2D Object Reference=@n2)
   n3 = GetComponentBounds!(Component:Scene Component Object Reference=Get CaptureTarget()) -> (Origin:Vector, BoxExtent:Vector, SphereRadius:float)
   n4 = GetForwardVector!(self:Scene Component Object Reference=Get CaptureCamera()) -> (ReturnValue:Vector)
   Set World Location!(
    execute=@n1.then_1
    self:Scene Component Object Reference=Get CaptureCamera()
    NewLocation:Vector=vector + vector(A=vector * vector(A=@n4, B:float=-100) -> (ReturnValue=~), B=@n3)
    bSweep:bool=false
    bTeleport:bool=true
   ) -> (then, SweepHitResult:Hit Result Structure) {is=title K2Node_CallFunction}
   n5 = Make Literal Float(Value=40)
   n6 = "Clamp (Float)"(Value=float - float(A=@n5, B=float * float(A=@n3.SphereRadius, B=2) {is=title K2Node_PromotableOperator}), Min=0, Max=@n5)
   > Set OrthoWidth!(
    execute
    OrthoWidth:float=float + float(A=float * float(A=@n3.SphereRadius, B=2), B=float + float(A=@n6, B=1))
    self:Scene Capture Component 2D Object Reference=Get CaptureCamera()
   ) -> (then, Output_Get:float)
   SetStaticMesh(execute=@n1.then_3, self=Get CaptureTarget(), NewMesh=Get DefaultPreviewMesh() -> (DefaultPreviewMesh:Static Mesh Object Reference))
   > Set Relative Transform!(
    execute
    self:Scene Component Object Reference=Get CaptureCamera()
    NewTransform:"Transform (by ref)"=Get DefaultCameraOffset() -> (DefaultCameraOffset:Transform)
    bSweep:bool=false
    bTeleport:bool=true
   ) -> (then, SweepHitResult:Hit Result Structure) {is=title K2Node_CallFunction}
   > Set Relative Location And Rotation!(
    execute
    self:Scene Component Object Reference=Get CaptureTarget()
    NewLocation:Vector="0, 0, 0"
    NewRotation:Rotator="0, 0, 0"
    bSweep:bool=false
    bTeleport:bool=true
   ) -> (then, SweepHitResult:Hit Result Structure) {is=title K2Node_CallFunction}
   n7 = Sequence!(execute=@n1.then_0) -> (then_0, then_1)
   n8 = Get CaptureTarget()
   > SetStaticMesh(self=@n8, NewMesh=Get StaticMeshAsset() -> (StaticMeshAsset:Static Mesh Object Reference))
   > n9 = For Each Loop(Array:Array of Material Interface Object References=GetMaterials!(self:Mesh Component Object Reference=@n8) -> (ReturnValue:Array of Material Interface Object References)) -> (Array Element:Material Interface Object Reference)
   > n10 = Get!(execute:Exec) -> (then:Exec, else:Exec, PreviewMeshMaterialObject:Material Interface Object Reference)
   > SetMaterial(self=@n8, ElementIndex=@n9.Array Index, Material=@n10)
   n11 = Get!() -> (AssetOffset_Location:Vector, AssetOffset_Rotation:Rotator, AssetOffset_Scale:Vector)
   Set Relative Location And Rotation(execute=@n7.then_1, self=Get CaptureTarget(), NewLocation=@n11, NewRotation=@n11.AssetOffset_Rotation, bSweep=false, bTeleport=true)
   n12 = GetDisplayName(Object=Get StaticMeshAsset())
   n13 = Replace(SourceString=@n12, From=_DECO)
   n14 = FindSubstring(SearchIn=@n13, Substring=VX)
   n15 = int - int(A=Len(S=@n12), B=@n14)
   Set ThumbnailName(
    execute=@n1.then_4
    ThumbnailName:String=Append(
     A=GetSubstring!(SourceString:String=@n13, StartIndex:int=@n14, Length:int=@n15) -> (ReturnValue:String)
     B=Get ThumbnailNameSurfix() -> (ThumbnailNameSurfix:String)
    )
   ) -> (Output_Get:String)
   n16 = Get ThumbnailRenderTarget() -> (ThumbnailRenderTarget:Texture Render Target 2D Object Reference)
   Set LODGroup!(execute=@n1.then_5, LODGroup:TextureGroup Enum=TEXTUREGROUP_UI, self:Texture Object Reference=@n16) -> (then, Output_Get:TextureGroup Enum)
   n17 = Get OutputDirectory() -> (OutputDirectory:String)
   > n18 = Render Target 2D Create Static Texture 2D Editor Only!(
    execute
    RenderTarget:Texture Render Target 2D Object Reference=@n16
    Name:String=Append!(A=NormalizeDirectoryName!(InPath:String=@n17) -> (OutPath:String), B=/, C:String=Get ThumbnailName() -> (ThumbnailName:String)) -> (ReturnValue)
    CompressionSettings:TextureCompressionSettings Enum=TC_EditorIcon
    MipSettings:TextureMipGenSettings Enum=TMGS_NoMipmaps
   ) -> (then, ReturnValue:Texture 2D Object Reference)
   > Set LODGroup(LODGroup=TEXTUREGROUP_UI, self=@n18)
   > Return Node!(execute, ThumbnailTexture:Texture 2D Object Reference=@n18)
  ]}
 EventGraph {nodes [
   n1 = Event GenerateThumbnails!() -> (OutputDelegate, then, BrickLibraryDataTable:Data Table Object Reference) {is=title K2Node_Event}
   > n2 = Sequence()
   > SetupAndLoadMaterials()
   GenerateThumbnailsAndUpdateDTReferences!(execute=@n2.then_1, self, PrimitiveDataTable:Data Table Object Reference=@n1.BrickLibraryDataTable) -> (then)
   Event BeginPlay()
   Event ActorBeginOverlap()
   Event Tick!() -> (OutputDelegate, then, DeltaSeconds:float)
  ]}
 GenerateThumbnailsAndUpdateDTReferences {nodes [
   GenerateThumbnailsAndUpdateDTReferences!() -> (then, PrimitiveDataTable:Data Table Object Reference) {is=title K2Node_FunctionEntry}
   > n1 = Sequence!(execute) -> (then_0, then_1, then_2:Exec, then_3:Exec)
   n2 = GetPathNameForLoadedAsset(execute=@n1.then_1, LoadedAsset=Get PrimitiveDataTable() -> (PrimitiveDataTable:Data Table Object Reference))
   > Set ThumbnailOutputDirectory(ThumbnailOutputDirectory:String=Append(
     A=Split!(InPath:String=@n2) -> (PathPart:String, FilenamePart:String, ExtensionPart:String)
     B=/
     C=Get ThumbnailDirectoryName() -> (ThumbnailDirectoryName:String)
    )) -> (Output_Get:String)
   n3 = GetDataTableRowNames(execute=@n1.then_2, Table=Get PrimitiveDataTable())
   > n4 = For Each Loop(Array:Array of Names=@n3) -> (Array Element:Name)
   > n5 = Get Data Table Row!(execute:Exec, DataTable:Data Table Object Reference=Get PrimitiveDataTable(), RowName:Name=@n4) -> (then:Exec, RowNotFound:Exec, ReturnValue:Atom Primitive Mapping Structure)
   n6 = Break Atom Primitive Mapping!(AtomPrimitiveMapping:"Atom Primitive Mapping Structure (by ref)"=@n5) -> (StaticMesh:Static Mesh Soft Object Reference, ThumbnailPrimitiveOffset:Transform) {is=title K2Node_BreakStruct}
   > LoadAsset_Blocking(Asset=@n6)
   > n7 = CaptureThumbnail!(
    execute
    self
    StaticMeshAsset:Static Mesh Object Reference=Resolve Soft Reference!(Input:Static Mesh Soft Object Reference=@n6) -> (Output:Static Mesh Object Reference)
    AssetOffset:Transform=@n6.ThumbnailPrimitiveOffset
    OutputDirectory:String=Get ThumbnailOutputDirectory() -> (ThumbnailOutputDirectory:String)
   ) -> (then, ThumbnailTexture:Texture 2D Object Reference) {is=title K2Node_CallFunction}
   > n8 = Set members in Atom Primitive Mapping!(
    execute:Exec
    StructRef:"Atom Primitive Mapping Structure (by ref)"=@n5
    Thumbnail:Texture 2D Soft Object Reference=Make Soft Reference!(Input:Texture 2D Object Reference=@n7) -> (Output:Texture 2D Soft Object Reference)
   ) -> (then:Exec, StructOut:"Atom Primitive Mapping Structure (by ref)") {is=title K2Node_SetFieldsInStruct}
   > AddDataTableRow!(
    execute
    DataTable:Data Table Object Reference=Get PrimitiveDataTable()
    RowName:"Name (by ref)"=@n4
    RowData:"Atom Primitive Mapping Structure (by ref)"=@n8
   ) -> (then)
  ]}
 SetupAndLoadMaterials {nodes [
   SetupAndLoadMaterials() {is=title K2Node_FunctionEntry}
   > n1 = Sequence!(execute) -> (then_0, then_1)
   n2 = LoadAsset_Blocking(execute=@n1.then_1, Asset=Get PostProcessMaterial() -> (PostProcessMaterial:Material Interface Soft Object Reference))
   n3 = Make Weighted Blendable!(Weight:float=1, Object:Object Reference=@n2) -> (WeightedBlendable:Weighted Blendable Structure) {is=title K2Node_MakeStruct}
   > Set members in Post Process Settings!(
    execute:Exec
    StructRef:"Post Process Settings Structure (by ref)"=Get Settings!(self:Post Process Component Object Reference=Get PostProcess() -> (PostProcess:Post Process Component Object Reference)) -> (Settings:Post Process Settings Structure)
    WeightedBlendables:Weighted Blendables Structure=Make Weighted Blendables!(Array:Array of Weighted Blendable Structures=Make Array("[0]":Weighted Blendable Structure=@n3) -> (Array:Array of Weighted Blendable Structures)) -> (WeightedBlendables:Weighted Blendables Structure) {is=title K2Node_MakeStruct}
   ) -> (then:Exec, StructOut:"Post Process Settings Structure (by ref)") {is=title K2Node_SetFieldsInStruct}
   n4 = LoadAsset_Blocking(execute=@n1.then_0, Asset=Get PreviewMeshMaterial() -> (PreviewMeshMaterial:Material Interface Soft Object Reference))
   > n5 = Cast To MaterialInterface!(execute, Object=@n4) -> (then, CastFailed, AsMaterial Interface:Material Interface Object Reference) {is=title K2Node_DynamicCast}
   > Set PreviewMeshMaterialObject(PreviewMeshMaterialObject:Material Interface Object Reference=@n5) -> (Output_Get:Material Interface Object Reference)
  ]}
 UserConstructionScript {nodes [
   Construction Script()
   > SetupAndLoadMaterials() {is=title K2Node_CallFunction}
  ]}
]
variables [
 DefaultCameraOffset {type=Transform, default {location=<0 0 0>, rotation=<rot -35 -30 0>, scale=<1 1 1>}}
 DefaultPreviewMesh {type=StaticMesh, default=null}
 PostProcessMaterial {type=MaterialInterface, default=&/PPM_Thumbnail}
 PreviewMeshMaterial {type=MaterialInterface, default=&/M_ThumbnailPreviewMeshMaterial}
 PreviewMeshMaterialObject {type=MaterialInterface, default=&M_ThumbnailPreviewMeshMaterial}
 ThumbnailDirectoryName {type=string, default=ElementIcons}
 ThumbnailName {type=string, default=""}
 ThumbnailNameSurfix {type=string, default=_icon}
 ThumbnailRenderTarget {type=TextureRenderTarget2D, default=&/ThumbnailRenderTarget}
]
