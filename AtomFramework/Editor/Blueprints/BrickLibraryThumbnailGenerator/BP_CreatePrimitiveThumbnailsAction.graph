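# Editor utility that generates thumbnails for brick library primitive Data Tables.
# GeneratePrimitiveThumbnails processes selected Data Table assets with a chosen generator class.
# The CreateThumbnailsAndUpdateBrickLibrary event passes its table along but hardcodes BP_DefaultAtomPrimitiveThumbnailGenerator, ignoring its class input.
# Both feed CreateThumbnailsFromGenerator, which spawns a transient generator actor, sends it the GenerateThumbnails interface message, then destroys it.
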EditorUtilityBlueprint /AtomFramework/Editor/Blueprints/BrickLibraryThumbnailGenerator/BP_CreatePrimitiveThumbnailsAction
graphs [
 CreateThumbnailsFromGenerator {nodes [
   n1 = CreateThumbnailsFromGenerator!() -> (then, PrimitiveDataTable:Data Table Object Reference, ThumbnailGeneratorClass:Actor Class Reference) {is=title K2Node_FunctionEntry}
   > n2 = Sequence!(execute) -> (then_0, then_1, then_2:Exec, then_3:Exec)
   n3 = Get!(execute:Exec=@n2.then_1) -> (then:Exec, else:Exec, PrimitiveThumbnailGenerator:Actor Object Reference)
   > GenerateThumbnails!(execute:Exec, self:Object Reference=@n3, BrickLibraryDataTable:Data Table Object Reference=@n1) -> (then:Exec) {is=title K2Node_Message}
   Destroy Actor(execute=@n2.then_3, self=Get PrimitiveThumbnailGenerator() -> (PrimitiveThumbnailGenerator:Actor Object Reference))
   n4 = SpawnActorFromClass!(
    execute=@n2.then_0
    ActorClass:Actor Class Reference=Get ThumbnailGeneratorClass() -> (ThumbnailGeneratorClass:Actor Class Reference)
    Location:Vector="0, 0, 0"
    Rotation:Rotator
    bTransient:bool=true
   ) -> (then, ReturnValue:Actor Object Reference)
   > Set PrimitiveThumbnailGenerator(PrimitiveThumbnailGenerator:Actor Object Reference=@n4) -> (Output_Get:Actor Object Reference)
  ]}
 EventGraph {nodes [
   n1 = Event CreateThumbnailsAndUpdateBrickLibrary!() -> (OutputDelegate, then, BrickLibraryDataTable:Data Table Object Reference, ThumbnailGeneratorClass:Actor Class Reference) {is=title K2Node_Event}
   > CreateThumbnailsFromGenerator!(
    execute
    self
    PrimitiveDataTable:Data Table Object Reference=@n1.BrickLibraryDataTable
    ThumbnailGeneratorClass:Actor Class Reference=&/BP_DefaultAtomPrimitiveThumbnailGenerator.BP_DefaultAtomPrimitiveThumbnailGenerator_C
   ) -> (then) {is=title K2Node_CallFunction}
  ]}
 GeneratePrimitiveThumbnails {nodes [
   GeneratePrimitiveThumbnails!() -> (then, ThumbnailsGeneratorClass:Actor Class Reference) {is=title K2Node_FunctionEntry}
   > n1 = GetSelectedAssets()
   > n2 = For Each Loop(Array:Array of Object References=@n1) -> (Array Element:Object Reference)
   > n3 = Cast To DataTable!(execute, Object=@n2) -> (then, CastFailed, AsData Table:Data Table Object Reference) {is=title K2Node_DynamicCast}
   > CreateThumbnailsFromGenerator(PrimitiveDataTable=@n3, ThumbnailGeneratorClass=Get ThumbnailsGeneratorClass() -> (ThumbnailsGeneratorClass:Actor Class Reference))
  ]}
]
variables []
